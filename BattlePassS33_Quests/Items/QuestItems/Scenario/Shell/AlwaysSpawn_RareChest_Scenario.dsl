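; Rare chest Blueprint for a Battle Pass quest scenario, a child class that gives the parent chest its blue rare look and opening burst.
; The construction script runs the parent's, then tints the glow and smoke sheet through material parameters and custom primitive data, and sets the rare visuals as the custom Niagara effects.
; The chest-open VFX event spawns the rare burst Niagara system at the static mesh transform.
; BeginPlay only calls the parent; overlap and tick are empty.
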(event EventBeginPlay
  (|CallParentFunction))

(event Collision|EventActorBeginOverlap
  (params (OtherActor :type Actor%20Object%20Reference)))

(event EventTick
  (params (DeltaSeconds :type Float%20%28single-precision%29)))

(event EventSpawnChestOpenVFX
  (Niagara|SpawnSystematLocation "/Game/Effects/Fort_Effects/World/NewChests/NS_ChestBurst_Rare.NS_ChestBurst_Rare" (.location (Transformation|GetWorldTransform (Variables|Debug|GetStaticMeshComponent))) (.rotation (Transformation|GetWorldTransform (Variables|Debug|GetStaticMeshComponent)))))

(fn ConstructionScript
  (bind _fx_glow___smoke_sheet (Variables|Default|GetFX%20Glow%20&%20Smoke%20Sheet))
  (|CallParentFunction)
  (Rendering|Material|SetDefaultCustomPrimitiveDataFloat _fx_glow___smoke_sheet 0 1.0) ; WPO_HideMesh
  (Rendering|Material|SetScalarParameterValueonMaterials _fx_glow___smoke_sheet "Corona_Opacity" 2e+01)
  (Rendering|Material|SetScalarParameterValueonMaterials _fx_glow___smoke_sheet "Glow_Opacity" 7.0)
  (Rendering|Material|SetVectorParameterValueonMaterials _fx_glow___smoke_sheet "Corona_Colour" "0.450000,0.874998,3.000000")
  (Rendering|Material|SetVectorParameterValueonMaterials _fx_glow___smoke_sheet "Glow_Colour" "0.050000,0.240000,1.000000")
  (Rendering|Material|SetScalarParameterValueonMaterials _fx_glow___smoke_sheet "UseBaseColor" 0.0)
  (Rendering|Material|SetColorParameterValueonMaterials _fx_glow___smoke_sheet "SecColor" "(R=0.000000,G=0.268970,B=1.000000,A=1.000000)")
  (Rendering|Material|SetColorParameterValueonMaterials _fx_glow___smoke_sheet "SpriteColor_Low" "(R=0.000000,G=0.369341,B=1.000000,A=1.000000)")
  (Variables|Default|SetCustom%20Niagara%20Effects (Utilities|Array|MakeArray (Variables|Default|GetRare%20Chest%20Visuals))))
