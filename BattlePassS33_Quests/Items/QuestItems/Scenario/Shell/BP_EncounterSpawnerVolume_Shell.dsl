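; BP_EncounterSpawnerVolume_Shell is a placeholder Blueprint, by its name an encounter spawner volume for quest scenarios, with no spawning logic in this graph.
; BeginPlay and EndPlay each print a "Hello" debug text.
; The overlap and tick events are empty stubs, and the construction script does nothing.
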(event EventEndPlay
  (params (EndPlayReason :type EEndPlayReason%20Enum :default "Destroyed"))
  (Development|PrintText "INVTEXT(\"Hello\")"))

(event EventBeginPlay
  (Development|PrintText "INVTEXT(\"Hello\")"))

(event Collision|EventActorBeginOverlap
  (params (OtherActor :type Actor%20Object%20Reference)))

(event EventTick
  (params (DeltaSeconds :type Float%20%28single-precision%29)))

(fn ConstructionScript)
