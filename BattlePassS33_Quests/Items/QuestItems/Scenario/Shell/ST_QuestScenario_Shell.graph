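# StateTree for the Season 33 battle pass Shell quest scenario, running linearly from setup through completion.
# Init spawns five magic splinters at tagged anchors, then advances once any player is present.
# Two stages each grant a playspace quest and wait on its tagged completion, the second also spawning a rare chest.
# A PoolSwirl NPC speech bubble and delay precede a Completed verb, success VFX and audio for participants, and tree success.
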StateTree /BattlePassS33_Quests/Items/QuestItems/Scenario/Shell/ST_QuestScenario_Shell
states [
 Root {
  id=#1
  type=STATE
  children [
   #2
   #3
   #4
   #5
   #6
   #7
  ]
  tasks []
  transitions []
 }
 Init {
  id=#2
  type=STATE
  tasks [
   {
    id=#8
    description=Spawn Actor
    properties {
     iD=#9
     instance {
      _structType=/Script/EncountersRuntime.SpawnActorStateTreeTaskInstanceData
      actorClass=&/BP_Shell_Scenario_MagicSplinter.BP_Shell_Scenario_MagicSplinter_C
      bAttachEncounterVolume=false
      bEnablePhysics=false
      bUseActorLocation=false
      bWaitForSpawnedActor=false
      owningActor=null
      persistenceBehavior=Destroy
      spawnedActor=null
      targetActor=null
      userActor=null
      actorIdentifierTags {gameplayTags []}
      actorTags {gameplayTags [{tagName=QuestEncounter.Shell.Shards.1}]}
      anchorOffset=<0 0 0>
      anchorQuery {
       autoDescription=" ANY( QuestEncounter.Shell.Shards.1 )"
       tokenStreamVersion=0
       userDescription=""
       queryTokenStream [0,1,1,1,0]
       tagDictionary [{tagName=QuestEncounter.Shell.Shards.1}]
      }
      existingActorQuery {autoDescription="", tokenStreamVersion=0, userDescription="", queryTokenStream [], tagDictionary []}
     }
     node {
      name=None
      _structType=/Script/EncountersRuntime.SpawnActorStateTreeTask
      bCanEditConsideredForCompletion=true
      bConsideredForCompletion=true
      bHasTaskCompletionDelegateDispatcher=false
      bTaskEnabled=true
      transitionHandlingPriority=Normal
      bindingsBatch {value=65535}
      executionRuntimeTemplateIndex {value=65535}
      instanceDataHandle {index=65535, source=None, stateHandle {index=65535}}
      instanceTemplateIndex {value=65535}
      outputBindingsBatch {value=65535}
     }
    }
   }
   {
    id=#10
    description=Spawn Actor
    properties {
     iD=#11
     instance {
      _structType=&SpawnActorStateTreeTaskInstanceData
      actorClass=&BP_Shell_Scenario_MagicSplinter_C
      bAttachEncounterVolume=false
      bEnablePhysics=false
      bUseActorLocation=false
      bWaitForSpawnedActor=false
      owningActor=null
      persistenceBehavior=Destroy
      spawnedActor=null
      targetActor=null
      userActor=null
      actorIdentifierTags {gameplayTags []}
      actorTags {gameplayTags [{tagName=QuestEncounter.Shell.Shards.2}]}
      anchorOffset=<0 0 0>
      anchorQuery {
       autoDescription=" ANY( QuestEncounter.Shell.Shards.2 )"
       tokenStreamVersion=0
       userDescription=""
       queryTokenStream [0,1,1,1,0]
       tagDictionary [{tagName=QuestEncounter.Shell.Shards.2}]
      }
      existingActorQuery {autoDescription="", tokenStreamVersion=0, userDescription="", queryTokenStream [], tagDictionary []}
     }
     node {
      name=None
      _structType=&SpawnActorStateTreeTask
      bCanEditConsideredForCompletion=true
      bConsideredForCompletion=true
      bHasTaskCompletionDelegateDispatcher=false
      bTaskEnabled=true
      transitionHandlingPriority=Normal
      bindingsBatch {value=65535}
      executionRuntimeTemplateIndex {value=65535}
      instanceDataHandle {index=65535, source=None, stateHandle {index=65535}}
      instanceTemplateIndex {value=65535}
      outputBindingsBatch {value=65535}
     }
    }
   }
   {
    id=#12
    description=Spawn Actor
    properties {
     iD=#13
     instance {
      _structType=&SpawnActorStateTreeTaskInstanceData
      actorClass=&BP_Shell_Scenario_MagicSplinter_C
      bAttachEncounterVolume=false
      bEnablePhysics=false
      bUseActorLocation=false
      bWaitForSpawnedActor=false
      owningActor=null
      persistenceBehavior=Destroy
      spawnedActor=null
      targetActor=null
      userActor=null
      actorIdentifierTags {gameplayTags []}
      actorTags {gameplayTags [{tagName=QuestEncounter.Shell.Shards.3}]}
      anchorOffset=<0 0 0>
      anchorQuery {
       autoDescription=" ANY( QuestEncounter.Shell.Shards.3 )"
       tokenStreamVersion=0
       userDescription=""
       queryTokenStream [0,1,1,1,0]
       tagDictionary [{tagName=QuestEncounter.Shell.Shards.3}]
      }
      existingActorQuery {autoDescription="", tokenStreamVersion=0, userDescription="", queryTokenStream [], tagDictionary []}
     }
     node {
      name=None
      _structType=&SpawnActorStateTreeTask
      bCanEditConsideredForCompletion=true
      bConsideredForCompletion=true
      bHasTaskCompletionDelegateDispatcher=false
      bTaskEnabled=true
      transitionHandlingPriority=Normal
      bindingsBatch {value=65535}
      executionRuntimeTemplateIndex {value=65535}
      instanceDataHandle {index=65535, source=None, stateHandle {index=65535}}
      instanceTemplateIndex {value=65535}
      outputBindingsBatch {value=65535}
     }
    }
   }
   {
    id=#14
    description=Spawn Actor
    properties {
     iD=#15
     instance {
      _structType=&SpawnActorStateTreeTaskInstanceData
      actorClass=&BP_Shell_Scenario_MagicSplinter_C
      bAttachEncounterVolume=false
      bEnablePhysics=false
      bUseActorLocation=false
      bWaitForSpawnedActor=false
      owningActor=null
      persistenceBehavior=Destroy
      spawnedActor=null
      targetActor=null
      userActor=null
      actorIdentifierTags {gameplayTags []}
      actorTags {gameplayTags [{tagName=QuestEncounter.Shell.Shards.4}]}
      anchorOffset=<0 0 0>
      anchorQuery {
       autoDescription=" ANY( QuestEncounter.Shell.Shards.4 )"
       tokenStreamVersion=0
       userDescription=""
       queryTokenStream [0,1,1,1,0]
       tagDictionary [{tagName=QuestEncounter.Shell.Shards.4}]
      }
      existingActorQuery {autoDescription="", tokenStreamVersion=0, userDescription="", queryTokenStream [], tagDictionary []}
     }
     node {
      name=None
      _structType=&SpawnActorStateTreeTask
      bCanEditConsideredForCompletion=true
      bConsideredForCompletion=true
      bHasTaskCompletionDelegateDispatcher=false
      bTaskEnabled=true
      transitionHandlingPriority=Normal
      bindingsBatch {value=65535}
      executionRuntimeTemplateIndex {value=65535}
      instanceDataHandle {index=65535, source=None, stateHandle {index=65535}}
      instanceTemplateIndex {value=65535}
      outputBindingsBatch {value=65535}
     }
    }
   }
   {
    id=#16
    description=Spawn Actor
    properties {
     iD=#17
     instance {
      _structType=&SpawnActorStateTreeTaskInstanceData
      actorClass=&BP_Shell_Scenario_MagicSplinter_C
      bAttachEncounterVolume=false
      bEnablePhysics=false
      bUseActorLocation=false
      bWaitForSpawnedActor=false
      owningActor=null
      persistenceBehavior=Destroy
      spawnedActor=null
      targetActor=null
      userActor=null
      actorIdentifierTags {gameplayTags []}
      actorTags {gameplayTags [{tagName=QuestEncounter.Shell.Shards.5}]}
      anchorOffset=<0 0 0>
      anchorQuery {
       autoDescription=" ANY( QuestEncounter.Shell.Shards.5 )"
       tokenStreamVersion=0
       userDescription=""
       queryTokenStream [0,1,1,1,0]
       tagDictionary [{tagName=QuestEncounter.Shell.Shards.5}]
      }
      existingActorQuery {autoDescription="", tokenStreamVersion=0, userDescription="", queryTokenStream [], tagDictionary []}
     }
     node {
      name=None
      _structType=&SpawnActorStateTreeTask
      bCanEditConsideredForCompletion=true
      bConsideredForCompletion=true
      bHasTaskCompletionDelegateDispatcher=false
      bTaskEnabled=true
      transitionHandlingPriority=Normal
      bindingsBatch {value=65535}
      executionRuntimeTemplateIndex {value=65535}
      instanceDataHandle {index=65535, source=None, stateHandle {index=65535}}
      instanceTemplateIndex {value=65535}
      outputBindingsBatch {value=65535}
     }
    }
   }
   {
    id=#18
    description="Scenario [Player Count Transition]"
    properties {
     iD=#19
     instance {
      _structType=/Script/EncountersRuntime.EncounterPlayerCountTransitionStateTreeTaskInstanceData
      bOnlyTriggerOnce=true
      bShouldTransition=true
      bTriggered=false
      comparisonOperator=GreaterThan
      userActor=null
      value=0
      customVerb {amount=1, customVerbTag {tagName=None}, subjects [], values []}
      onTriggered {iD=#20}
     }
     node {
      name=None
      _structType=/Script/EncountersRuntime.EncounterPlayerCountTransitionStateTreeTask
      bCanEditConsideredForCompletion=true
      bConsideredForCompletion=true
      bHasTaskCompletionDelegateDispatcher=false
      bTaskEnabled=true
      transitionHandlingPriority=Normal
      bindingsBatch {value=65535}
      executionRuntimeTemplateIndex {value=65535}
      instanceDataHandle {index=65535, source=None, stateHandle {index=65535}}
      instanceTemplateIndex {value=65535}
      outputBindingsBatch {value=65535}
      transitionTo {name=None, fallback=None, iD=#20, linkType=NextSelectableState, stateHandle {index=65535}}
     }
    }
   }
  ]
  transitions []
 }
 Stage_Friendly {
  id=#3
  type=STATE
  tasks [
   {
    id=#21
    description=Grant Quest
    properties {
     iD=#22
     instance {
      _structType=/Script/EncountersRuntime.GrantQuestStateTreeTaskInstanceData
      grantType=GrantToPlayspace
      questToGrant=/BattlePassS33_Quests/Items/QuestItems/Scenario/Shell/Quests/PlaySpace_EncounterQuest_Shell_00
      targetActor=null
      userActor=null
     }
     node {
      name=None
      _structType=/Script/EncountersRuntime.GrantQuestStateTreeTask
      bCanEditConsideredForCompletion=true
      bConsideredForCompletion=true
      bHasTaskCompletionDelegateDispatcher=false
      bTaskEnabled=true
      transitionHandlingPriority=Normal
      bindingsBatch {value=65535}
      executionRuntimeTemplateIndex {value=65535}
      instanceDataHandle {index=65535, source=None, stateHandle {index=65535}}
      instanceTemplateIndex {value=65535}
      outputBindingsBatch {value=65535}
     }
    }
   }
   {
    id=#23
    description=Scenario Verb Transition
    properties {
     iD=#24
     instance {
      _structType=/Script/EncountersRuntime.EncounterVerbTransitionStateTreeTaskInstanceData
      bIsTriggered=false
      bUseAdvancedObjectives=false
      objectiveValueThreshold=1
      userActor=null
      objectiveVerbs [{
        _structType=/Script/FortniteGame.FortObjectiveFilter_QuestUpdated
        bUseQuestDefinitionsFilter=false
        questState=Completed
        questSubtype=None
        questType=All
        author {
         bCanGainProgressForThisObjective=true
         tagQuery {autoDescription="", tokenStreamVersion=0, userDescription="", queryTokenStream [], tagDictionary []}
        }
        context {tagQuery {autoDescription="", tokenStreamVersion=0, userDescription="", queryTokenStream [], tagDictionary []}}
        questDefinitions {advancedQuestFilters []}
        questTags {tagQuery {
          autoDescription=" ANY( QuestEncounter.Shell.Quest.01 )"
          tokenStreamVersion=0
          userDescription=""
          queryTokenStream [0,1,1,1,0]
          tagDictionary [{tagName=QuestEncounter.Shell.Quest.01}]
         }}
       }]
      objectives []
      onTriggered {iD=#20}
     }
     node {
      name=None
      _structType=/Script/EncountersRuntime.EncounterVerbTransitionStateTreeTask
      bCanEditConsideredForCompletion=true
      bConsideredForCompletion=true
      bHasTaskCompletionDelegateDispatcher=false
      bTaskEnabled=true
      transitionHandlingPriority=Normal
      bindingsBatch {value=65535}
      executionRuntimeTemplateIndex {value=65535}
      instanceDataHandle {index=65535, source=None, stateHandle {index=65535}}
      instanceTemplateIndex {value=65535}
      outputBindingsBatch {value=65535}
      transitionTo {name=None, fallback=None, iD=#20, linkType=NextSelectableState, stateHandle {index=65535}}
     }
    }
   }
  ]
  transitions []
 }
 NPC Reward Objective {
  id=#4
  type=STATE
  tasks [
   {
    id=#25
    description=Grant Quest
    properties {
     iD=#26
     instance {
      _structType=&GrantQuestStateTreeTaskInstanceData
      grantType=GrantToPlayspace
      questToGrant=/BattlePassS33_Quests/Items/QuestItems/Scenario/Shell/Quests/PlaySpace_EncounterQuest_Shell_02
      targetActor=null
      userActor=null
     }
     node {
      name=None
      _structType=&GrantQuestStateTreeTask
      bCanEditConsideredForCompletion=true
      bConsideredForCompletion=true
      bHasTaskCompletionDelegateDispatcher=false
      bTaskEnabled=true
      transitionHandlingPriority=Normal
      bindingsBatch {value=65535}
      executionRuntimeTemplateIndex {value=65535}
      instanceDataHandle {index=65535, source=None, stateHandle {index=65535}}
      instanceTemplateIndex {value=65535}
      outputBindingsBatch {value=65535}
     }
    }
   }
   {
    id=#27
    description=Spawn Actor
    properties {
     iD=#28
     instance {
      _structType=&SpawnActorStateTreeTaskInstanceData
      actorClass=&/AlwaysSpawn_RareChest_Scenario.AlwaysSpawn_RareChest_Scenario_C
      bAttachEncounterVolume=false
      bEnablePhysics=false
      bUseActorLocation=false
      bWaitForSpawnedActor=false
      owningActor=null
      persistenceBehavior=Destroy
      spawnedActor=null
      targetActor=null
      userActor=null
      actorIdentifierTags {gameplayTags [{tagName=QuestEncounter.Shell.Reward.01}]}
      actorTags {gameplayTags []}
      anchorOffset=<0 0 0>
      anchorQuery {
       autoDescription=" ANY( QuestEncounter.Shell.Reward.01 )"
       tokenStreamVersion=0
       userDescription=""
       queryTokenStream [0,1,1,1,0]
       tagDictionary [{tagName=QuestEncounter.Shell.Reward.01}]
      }
      existingActorQuery {autoDescription="", tokenStreamVersion=0, userDescription="", queryTokenStream [], tagDictionary []}
     }
     node {
      name=None
      _structType=&SpawnActorStateTreeTask
      bCanEditConsideredForCompletion=true
      bConsideredForCompletion=true
      bHasTaskCompletionDelegateDispatcher=false
      bTaskEnabled=true
      transitionHandlingPriority=Normal
      bindingsBatch {value=65535}
      executionRuntimeTemplateIndex {value=65535}
      instanceDataHandle {index=65535, source=None, stateHandle {index=65535}}
      instanceTemplateIndex {value=65535}
      outputBindingsBatch {value=65535}
     }
    }
   }
   {
    id=#29
    description=Scenario Verb Transition
    properties {
     iD=#30
     instance {
      _structType=&EncounterVerbTransitionStateTreeTaskInstanceData
      bIsTriggered=false
      bUseAdvancedObjectives=false
      objectiveValueThreshold=1
      userActor=null
      objectiveVerbs [{
        _structType=&FortObjectiveFilter_QuestUpdated
        bUseQuestDefinitionsFilter=false
        questState=Completed
        questSubtype=None
        questType=All
        author {
         bCanGainProgressForThisObjective=true
         tagQuery {autoDescription="", tokenStreamVersion=0, userDescription="", queryTokenStream [], tagDictionary []}
        }
        context {tagQuery {autoDescription="", tokenStreamVersion=0, userDescription="", queryTokenStream [], tagDictionary []}}
        questDefinitions {advancedQuestFilters []}
        questTags {tagQuery {
          autoDescription=" ANY( QuestEncounter.Shell.Quest.02 )"
          tokenStreamVersion=0
          userDescription=""
          queryTokenStream [0,1,1,1,0]
          tagDictionary [{tagName=QuestEncounter.Shell.Quest.02}]
         }}
       }]
      objectives []
      onTriggered {iD=#20}
     }
     node {
      name=None
      _structType=&EncounterVerbTransitionStateTreeTask
      bCanEditConsideredForCompletion=true
      bConsideredForCompletion=true
      bHasTaskCompletionDelegateDispatcher=false
      bTaskEnabled=true
      transitionHandlingPriority=Normal
      bindingsBatch {value=65535}
      executionRuntimeTemplateIndex {value=65535}
      instanceDataHandle {index=65535, source=None, stateHandle {index=65535}}
      instanceTemplateIndex {value=65535}
      outputBindingsBatch {value=65535}
      transitionTo {name=None, fallback=None, iD=#20, linkType=NextSelectableState, stateHandle {index=65535}}
     }
    }
   }
  ]
  transitions []
 }
 Reward Drop Reaction {
  id=#5
  type=STATE
  tasks [
   {
    id=#31
    description=Get Scenario Actor
    properties {
     iD=#32
     instance {
      _structType=&GetEncounterActorStateTreeTaskInstanceData
      actorQuery {autoDescription=" ANY( AISpawnerData.Type.Tandem.PoolSwirl )", tagDictionary [{tagName=AISpawnerData.Type.Tandem.PoolSwirl}]}
     }
     node {name=None, _structType=&GetEncounterActorStateTreeTask}
    }
   }
   {
    id=#33
    description=Grant Reactions
    properties {
     iD=#34
     instance {
      _structType=/Script/EncountersRuntime.GrantReactionStateTreeTaskInstanceData
      grantType=GrantToTargetActor
      targetActor=null
      userActor=null
      reactionEntries []
      reactions [{
        _structType=/Script/FortniteGame.FortEventReaction_PlaySpeechBubble
        bIncludeVO=false
        bInterruptActiveSpeech=true
        speechBubbleTagQuery {
         autoDescription=" ANY( QuestEncounter.Shell.Dialog.Completion.02 )"
         tokenStreamVersion=0
         userDescription=""
         queryTokenStream [0,1,1,1,0]
         tagDictionary [{tagName=QuestEncounter.Shell.Dialog.Completion.02}]
        }
        vOPlaybackTags {gameplayTags []}
       }]
     }
     node {
      name=None
      _structType=/Script/EncountersRuntime.GrantReactionStateTreeTask
      bCanEditConsideredForCompletion=true
      bConsideredForCompletion=true
      bHasTaskCompletionDelegateDispatcher=false
      bTaskEnabled=true
      transitionHandlingPriority=Normal
      bindingsBatch {value=65535}
      executionRuntimeTemplateIndex {value=65535}
      instanceDataHandle {index=65535, source=None, stateHandle {index=65535}}
      instanceTemplateIndex {value=65535}
      outputBindingsBatch {value=65535}
     }
    }
   }
  ]
  transitions [{
    id=#35
    trigger=ON_STATE_COMPLETED
    state {id=#36, name=Root, linkType=NEXT_SELECTABLE_STATE}
   }]
 }
 Delay {
  id=#6
  type=STATE
  tasks [{
    id=#37
    description=Delay 3.0
    properties {
     iD=#38
     instance {_structType=&StateTreeDelayTaskInstanceData, duration=3}
     node {name=None, _structType=&StateTreeDelayTask}
    }
   }]
  transitions [{
    id=#39
    trigger=ON_STATE_COMPLETED
    state {id=#36, name=Root, linkType=NEXT_SELECTABLE_STATE}
   }]
 }
 Complete {
  id=#7
  type=STATE
  tasks [
   {
    id=#40
    description=Send Custom Verb
    properties {
     iD=#41
     instance {
      _structType=/Script/EncountersRuntime.EncounterSendCustomVerbStateTreeTaskInstanceData
      bSendVerbForEachPlayerAuthor=false
      userActor=null
      customVerb {amount=1, customVerbTag {tagName=CustomVerb.Encounter.Completed}, subjects [], values []}
     }
     node {
      name=None
      _structType=/Script/EncountersRuntime.EncounterSendCustomVerbStateTreeTask
      bCanEditConsideredForCompletion=true
      bConsideredForCompletion=true
      bHasTaskCompletionDelegateDispatcher=false
      bTaskEnabled=true
      transitionHandlingPriority=Normal
      bindingsBatch {value=65535}
      executionRuntimeTemplateIndex {value=65535}
      instanceDataHandle {index=65535, source=None, stateHandle {index=65535}}
      instanceTemplateIndex {value=65535}
      outputBindingsBatch {value=65535}
     }
    }
   }
   {
    id=#42
    description=BP SDA STTask Complete State
    properties {
     iD=#43
     instanceObject=&self:StateTreeEditorData_0.StateTreeState_7.BP_SDA_STTask_CompleteState_C_0
     node {
      name=None
      _structType=&StateTreeBlueprintTaskWrapper
      taskClass=/StateDrivenActor/Tasks/BP_SDA_STTask_CompleteState.BP_SDA_STTask_CompleteState_C
     }
    }
   }
   {
    id=#44
    description=Grant Reactions
    properties {
     iD=#45
     instance {
      _structType=&GrantReactionStateTreeTaskInstanceData
      grantType=GrantToParticipants
      targetActor=null
      userActor=null
      reactionEntries []
      reactions [
       {
        _structType=/Script/FortniteGame.FortEventReaction_PlayVFX
        bForceVFXAtTargetLocation=false
        vFX=/Game/Effects/Fort_Effects/Gameplay/TimeTrial/NS_S29_MythicLabors_FootRace_Success
        effectLocationOffset=<0 0 0>
        effectScale=<1 1 1>
       }
       {
        _structType=/Script/FortniteGame.FortEventReaction_PlayAudio
        audio=/Game/Athena/Sounds/Quests/TimedTrial/Athena_Quest_TimeTrial_Success_Cue
        bForceSoundAtTargetLocation=false
        pitchMultiplier=1
        volumeMultiplier=1
        soundLocationOffset=<0 0 0>
       }
      ]
     }
     node {
      name=None
      _structType=&GrantReactionStateTreeTask
      bCanEditConsideredForCompletion=true
      bConsideredForCompletion=true
      bHasTaskCompletionDelegateDispatcher=false
      bTaskEnabled=true
      transitionHandlingPriority=Normal
      bindingsBatch {value=65535}
      executionRuntimeTemplateIndex {value=65535}
      instanceDataHandle {index=65535, source=None, stateHandle {index=65535}}
      instanceTemplateIndex {value=65535}
      outputBindingsBatch {value=65535}
     }
    }
   }
  ]
  transitions [{id=#46, trigger=ON_STATE_COMPLETED, state {id=#36, name=Root, linkType=SUCCEEDED}}]
 }
]
